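# Material function that blends a glowing teal slurp look into incoming ocean material attributes.
# Slurp color lerps two color parameters by a swamp mask sampled at zeroed UVs, feeding base color and emissive scaled by time of day.
# The blend alpha is a soft sphere around the collection's slurp center and radius, cut by a world-projected, non-wrapping juice mask texture, then clamped.
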MaterialFunction /Clyde/Athena/Apollo/Water/Materials/MF_OceanSlurp
graph {nodes [
  AppendVector()
  n1 = NamedRerouteUsage()
  FunctionOutput(None=BlendMaterialAttributes(A=FunctionInput() {inputType=FunctionInput_MaterialAttributes}, B=NamedRerouteUsage(), Alpha=Clamp(None=@n1)) {vertexAttributeBlendType=UseA})
  n2 = TextureCoordinate()
  n3 = LinearInterpolate(
   A=VectorParameter() {parameterName=ColorA, defaultValue=<rgba 0.0008 0.1042 0.0874 1>}.RGB
   B=VectorParameter() {parameterName=ColorB, defaultValue=<rgba 0.0076 0.974 0.8173 1>}.RGB
   Alpha=TextureSample(UVs=Multiply(A=@n2) {constB=0}) {samplerType=SAMPLERTYPE_Masks, texture=/Game/Global/Textures/Default/Masks_and_Gradients/Swamp_Mask}.B
  )
  n4 = Multiply(A=@n3, B=ScalarParameter() {defaultValue=1, parameterName=Emissive Intensity})
  n5 = Constant() {r=1}
  n6 = MaterialFunctionCall(Color_Morning=@n5, Color_Day=@n5, Color_Evening=@n5, Color_Night=ScalarParameter() {defaultValue=0.2, parameterName=Night Emissive Multi}) {materialFunction=&Color_Time_of_Day}
  NamedRerouteDeclaration(None=SetMaterialAttributes!(MaterialAttributes, BaseColor=@n3, EmissiveColor=Multiply(A=@n4, B=@n6.Result)) {attributeSetTypes [#1, #2]}) {name=SlurpMaterial, nodeColor=<rgba 0.9453 0 1 1>}
  n7 = WorldPosition()
  n8 = ComponentMask(None=CollectionParameter() {collection=&FortniteMaterialParameters, parameterName=SlurpyCenter}) {b=true, g=true, r=true}
  n9 = Multiply(A=ComponentMask(None=CollectionParameter() {collection=&FortniteMaterialParameters, parameterName=SlurpyRadius}) {b=true, g=true, r=true}) {constB=1.5}
  n10 = CollectionParameter() {
   collection=&FortniteMaterialParameters
   parameterName=SlurpTextureMaskLocation
  }
  n11 = Add(A=WorldPosition().XYZ, B=OneMinus(None=ComponentMask(None=@n10) {b=true, g=true, r=true}))
  n12 = TextureObjectParameter() {parameterName=JuiceMask, samplerSource=SSM_Clamp_WorldGroupSettings, texture=&/Clyde_OilSpill}
  n13 = ComponentMask(None=MaterialFunctionCall!(
    Export Float 4=StaticBoolParameter() {defaultValue=true, parameterName="Export Float 4 (SB)"}
    TextureObject=@n12
    TextureSize=CollectionParameter() {collection=&FortniteMaterialParameters, parameterName=JuiceMaskSize}
    WorldPosition=@n11
    World Space Normal=@n11
    RotZAxis
    RotXAxis
    RotYAxis
    ProjectionTransitionContrast
   ) {materialFunction=&/MF_WorldAlignedTexture_NoWrap}.XYZ Texture) {a=true}
  NamedRerouteDeclaration(None=Subtract(A=SphereMask(A=@n7.XYZ, B=@n8, Radius=@n9) {hardnessPercent=5}, B=OneMinus(None=@n13))) {name=SlurpMask, nodeColor=<rgba 0.0442 0.7756 1 1>}
 ]}
