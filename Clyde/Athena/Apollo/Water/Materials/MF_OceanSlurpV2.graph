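# Material function that blends a glowing slurp layer into incoming ocean material attributes and also outputs a Slurp_Mask.
# A collection-driven sphere and world-aligned JuiceMask texture define the slurp area, while water depth shapes shoreline fade and seam.
# Panned, flow-distorted world-space noise drives alpha and color, with emissive scaled by time of day.
# Curve-atlas directional waves, limited by a sphere and damped at shore, are built alongside turbulence accents.
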MaterialFunction /Clyde/Athena/Apollo/Water/Materials/MF_OceanSlurpV2
graph {nodes [
  n1 = Time()
  n2 = WorldPosition()
  n3 = ScalarParameter() {defaultValue=0.0001, group=Slurp, parameterName=Flow Map Alpha Texture Scale}
  n4 = NamedRerouteUsage()
  n5 = ScalarParameter() {defaultValue=0.2, parameterName=Alpha Wave Distortion Strength}
  n6 = NamedRerouteUsage()
  n7 = ScalarParameter() {defaultValue=0.01, group=Slurp, parameterName=Alpha Texture Speed}
  n8 = Panner(Coordinate=Add(A=Multiply(A=@n2.XY, B=@n3), B=Multiply(A=@n4, B=@n5)), Time=Time(), Speed=Multiply(A=Multiply(A=@n6, B=@n7)) {constB=-1})
  n9 = WorldPosition()
  n10 = ScalarParameter() {defaultValue=0, group=Slurp, parameterName=Flow Map Alpha Distortion Scale}
  n11 = Multiply(A=NamedRerouteUsage(), B=ScalarParameter() {group=Slurp, parameterName=Flow Map Glow Alpha Motion Speed})
  n12 = TextureSample(
   UVs=Panner(Coordinate=Multiply(A=@n9.XY, B=@n10), Time=Time(), Speed=Multiply(A=@n11) {constB=-1})
   Tex=TextureObject() {samplerType=SAMPLERTYPE_LinearColor, texture=/Clyde/Athena/Apollo/Water/Textures/KelpLinen/KelpLinen_CloudFlowMap}
  ) {samplerSource=SSM_Wrap_WorldGroupSettings, samplerType=SAMPLERTYPE_LinearColor}
  n13 = Multiply(
   A=ConstantBiasScale(None=ComponentMask(None=@n12.RGB) {g=true, r=true}) {bias=-0.5, scale=2}
   B=ScalarParameter() {defaultValue=0.4, group=Slurp, parameterName=Flow Map Alpha Strength}
  )
  n14 = Multiply(A=Time(), B=ScalarParameter() {defaultValue=0.005, group=Slurp, parameterName=Flow Map Alpha Speed})
  n15 = TextureObject() {samplerType=SAMPLERTYPE_Masks, texture=/Game/Global/Textures/Default/Masks_and_Gradients/Swamp_Mask}
  n16 = TextureSample(UVs=Add(A=@n8, B=Add(A=@n13, B=@n14)), Tex=@n15) {samplerSource=SSM_Wrap_WorldGroupSettings}
  n17 = ComponentMask(None=@n16.RGB) {b=true}
  Add(A=@n1, B=Multiply(A=ConstantBiasScale(None=@n17) {bias=-0.5, scale=2}, B=ScalarParameter() {parameterName=Accent Distortion Strength}))
  n18 = FunctionInput() {inputType=FunctionInput_MaterialAttributes}
  n19 = NamedRerouteUsage()
  n20 = ScalarParameter() {defaultValue=1.5, group=Slurp, parameterName=Alpha Power Shoreline}
  n21 = LinearInterpolate(
   A=ScalarParameter() {defaultValue=4, group=Slurp, parameterName=Alpha Power Min}
   B=ScalarParameter() {defaultValue=1, group=Slurp, parameterName=Alpha Power Max}
   Alpha=NamedRerouteUsage()
  ) {constB=2}
  n22 = ScalarParameter() {defaultValue=0.2, group=Slurp, parameterName=Shoreline Alpha Depth Cutoff Min}
  n23 = ScalarParameter() {defaultValue=0.8, group=Slurp, parameterName=Shoreline Alpha Depth Cutoff Max}
  n24 = NamedRerouteUsage()
  n25 = Power(
   Base=Saturate(None=InverseLinearInterpolate(A=@n22, B=@n23, Value=@n24))
   Exp=ScalarParameter() {defaultValue=1, group=Slurp, parameterName=Shoreline Depth Power Alpha}
  )
  n26 = Multiply(
   A=Power(Base=NamedRerouteUsage(), Exp=LinearInterpolate(A=@n20, B=@n21, Alpha=@n25))
   B=ScalarParameter() {defaultValue=1, group=Slurp, parameterName=Alpha Strength}
  )
  n27 = Saturate(None=LinearInterpolate(A=ScalarParameter() {defaultValue=0.25, group=Slurp, parameterName=Alpha Low Min}, Alpha=Saturate(None=@n26)))
  n28 = NamedRerouteUsage()
  n29 = OneMinus(None=@n28)
  n30 = ScalarParameter() {defaultValue=0.05, group=Slurp, parameterName=Shoreline Seam Depth Cutoff Min}
  n31 = Divide(A=@n30) {constB=3}
  n32 = ScalarParameter() {defaultValue=0.1, group=Slurp, parameterName=Shorline Seam Depth Cutoff Max}
  n33 = Divide(A=@n32) {constB=3}
  n34 = Saturate(None=InverseLinearInterpolate(A=LinearInterpolate(A=@n30, B=@n31, Alpha=@n28) {constB=0.035}, B=LinearInterpolate(A=@n32, B=@n33, Alpha=@n28), Value=NamedRerouteUsage()))
  n35 = OneMinus(None=Power(Base=@n34, Exp=ScalarParameter() {defaultValue=0.5, group=Slurp, parameterName=Shoreline Seam Depth Noise Power}))
  n36 = Multiply(
   A=OneMinus(None=Multiply(A=@n29, B=@n35))
   B=Power(Base=@n34, Exp=ScalarParameter() {defaultValue=0.5, group=Slurp, parameterName=Shoreline Seam Depth Power})
  )
  n37 = OneMinus(None=@n36)
  n38 = Multiply(A=Saturate(None=Subtract(A=@n27, B=@n37)), B=@n36)
  n39 = Multiply(B=NamedRerouteUsage()) {constA=1}
  n40 = Saturate(None=Subtract(A=Saturate(None=@n38), B=OneMinus(None=@n39)))
  n41 = NamedRerouteDeclaration(None=Saturate(None=Multiply(A=@n40, B=@n39))) {name=Alpha Out, nodeColor=<rgba 1 0 0.1875 1>}
  BlendMaterialAttributes(A=@n18, B=@n19, Alpha=@n41)
  ComponentMask(None=NamedRerouteUsage()) {b=true, g=true, r=true}
  n42 = GetMaterialAttributes(None=@n18) {attributeGetTypes [
    #1
    #2
    #3
    #4
    #5
   ]}
  n43 = GetMaterialAttributes(None=@n19) {attributeGetTypes [
    #1
    #2
    #3
    #4
    #5
   ]}
  n44 = SetMaterialAttributes!(
   MaterialAttributes=@n42.MaterialAttributes
   BaseColor=LinearInterpolate(A=@n42.BaseColor, B=@n43.BaseColor, Alpha=@n41)
   EmissiveColor=LinearInterpolate(A=@n42.EmissiveColor, B=@n43.EmissiveColor, Alpha=@n41)
   Specular=LinearInterpolate(A=@n42.Specular, B=@n43.Specular, Alpha=@n41)
   Roughness=LinearInterpolate(A=@n42.Roughness, B=@n43.Roughness, Alpha=@n41)
   Normal=LinearInterpolate(A=@n42.Normal, B=@n43.Normal, Alpha=@n41)
  ) {attributeSetTypes [
    #1
    #2
    #3
    #4
    #5
   ]}
  FunctionOutput(None=@n44)
  FunctionOutput(None=NamedRerouteUsage()) {outputName=Slurp_Mask}
  n45 = NamedRerouteUsage()
  LinearInterpolate(A=Constant3Vector() {constant=<rgba 0 0 1 0>}, B=@n45, Alpha=Saturate(None=NamedRerouteUsage())) {constB=1.1682}
  n46 = TextureObject() {samplerType=SAMPLERTYPE_Masks, texture=&Swamp_Mask}
  n47 = WorldPosition()
  n48 = ScalarParameter() {defaultValue=0, group=Slurp, parameterName=Flow Map Glow Distortion Scale}
  n49 = Multiply(A=NamedRerouteUsage(), B=ScalarParameter() {defaultValue=0.05, group=Slurp, parameterName=Flow Map Glow Distortion Motion Speed})
  n50 = TextureSample(
   UVs=Panner(Coordinate=Multiply(A=@n47.XY, B=@n48), Time=Time(), Speed=Multiply(A=@n49) {constB=-1})
   Tex=TextureObject() {samplerType=SAMPLERTYPE_LinearColor, texture=&KelpLinen_CloudFlowMap}
  ) {samplerSource=SSM_Wrap_WorldGroupSettings, samplerType=SAMPLERTYPE_LinearColor}
  n51 = Multiply(
   A=ConstantBiasScale(None=ComponentMask(None=@n50.RGB) {g=true, r=true}) {bias=-0.5, scale=2}
   B=ScalarParameter() {defaultValue=1, group=Slurp, parameterName=Flow Map Glow Strength}
  )
  n52 = Multiply(A=Time(), B=ScalarParameter() {defaultValue=0.05, group=Slurp, parameterName=Flow Map Glow Speed})
  n53 = Add(
   A=Multiply(A=WorldPosition().XY, B=ScalarParameter() {defaultValue=0.0002, group=Slurp, parameterName=Flow Map Glow Texture Scale})
   B=Multiply(A=NamedRerouteUsage(), B=ScalarParameter() {defaultValue=0.2, parameterName=Glow Wave Distortion Strength})
  )
  MaterialFunctionCall!(
   Diffuse=@n46
   Normal Map
   "Flow Vector Map (see tooltip)"=@n51
   UV texture 2 offset
   "Time (see tooltip)"=@n52
   useMipLevel
   UVs=@n53
   Texture Size=TextureProperty(Tex=@n46)
   Texture Mip Bias
  ) {materialFunction=/Engine/Functions/Engine_MaterialFunctions02/Texturing/Flowmaps.FlowMaps}
  MaterialFunctionCall(Diffuse=@n15, "Flow Vector Map (see tooltip)"=@n13, "Time (see tooltip)"=@n14, UVs=@n8, Texture Size=TextureProperty(Tex=@n15)) {materialFunction=&FlowMaps}
  n54 = NamedRerouteUsage()
  n55 = Multiply(A=Multiply(A=ComponentMask(None=@n54) {g=true, r=true}) {constB=-1}, B=NamedRerouteUsage())
  Multiply(A=@n55, B=ScalarParameter() {defaultValue=0.2, group=Slurp, parameterName=Wave Distortion Strength})
  NamedRerouteDeclaration(None=MaterialFunctionCall(
    Fade Length=ScalarParameter() {defaultValue=43732.9492, group=Distance Blending, parameterName=Skydive_Distance}
    Fade Offset=ScalarParameter() {defaultValue=500, group=Distance Blending, parameterName=Skydive_Offset}
    For Vertex Shader=StaticBool() {value=true}
   ) {materialFunction=&CameraDepthFade}.Result) {name=Skydive_Fade, nodeColor=<rgba 0.4297 1 0 1>}
  NamedRerouteDeclaration(None=@n42.Normal) {name=Normal_ocean, nodeColor=<rgba 0 1 0.3672 1>}
  NamedRerouteDeclaration(None=ComponentMask(None=@n16.RGB) {r=true}) {name=Shoreline Noise, nodeColor=<rgba 0 1 0.4609 1>}
  n56 = Add(
   A=Multiply(A=WorldPosition().XY, B=ScalarParameter() {defaultValue=0.0035, group=Slurp, parameterName=Accent Texture Scale})
   B=Multiply(A=NamedRerouteUsage(), B=ScalarParameter() {defaultValue=2, parameterName=Accent Wave Distortion Strength})
  ) {constB=0}
  n57 = Multiply(A=Multiply(A=NamedRerouteUsage(), B=ScalarParameter() {defaultValue=0.1, group=Slurp, parameterName=Accent Texture Speed})) {constB=-1}
  n58 = TextureObject() {texture=/FortWater/Textures/T_Turbulence_Seamless}
  n59 = ComponentMask(None=TextureSample(UVs=Panner(Coordinate=@n56, Time=@n1, Speed=@n57), Tex=@n58) {samplerSource=SSM_Wrap_WorldGroupSettings, samplerType=SAMPLERTYPE_Masks, texture=/FortWater/Textures/T_OceanFoam_New_Mask}.RGB) {r=true}
  n60 = Add(
   A=Multiply(A=WorldPosition().XY, B=ScalarParameter() {defaultValue=0.0025, group=Slurp, parameterName=Accent Texture Scale 2})
   B=Multiply(A=NamedRerouteUsage(), B=ScalarParameter() {defaultValue=2, parameterName=Accent Wave Distortion Strength 2})
  ) {constB=0}
  n61 = Multiply(A=Multiply(A=NamedRerouteUsage(), B=ScalarParameter() {defaultValue=0.05, group=Slurp, parameterName=Accent Texture Speed 2})) {constB=-1}
  n62 = ComponentMask(None=TextureSample(UVs=Panner(Coordinate=@n60, Time=@n1, Speed=@n61), Tex=@n58) {samplerSource=SSM_Wrap_WorldGroupSettings, samplerType=SAMPLERTYPE_Masks, texture=&T_OceanFoam_New_Mask}.RGB) {r=true}
  NamedRerouteDeclaration(None=Saturate(None=LinearInterpolate(A=@n59, B=@n62) {constAlpha=0.75})) {name=Accent Out, nodeColor=<rgba 1 0 0.1875 1>}
  n63 = Constant() {r=0.06}
  n64 = Multiply(
   A=Power(Base=NamedRerouteUsage(), Exp=ScalarParameter() {defaultValue=2, group=Slurp, parameterName=Alpha Color Power})
   B=ScalarParameter() {defaultValue=1, group=Slurp, parameterName=Alpha Color Strength}
  )
  n65 = NamedRerouteUsage()
  n66 = NamedRerouteUsage()
  n67 = NamedRerouteUsage()
  n68 = LinearInterpolate(
   A=ScalarParameter() {defaultValue=3, group=Slurp, parameterName=Glow Power Min}
   B=ScalarParameter() {defaultValue=1, group=Slurp, parameterName=Glow Power Max}
   Alpha=LinearInterpolate(B=Saturate(None=@n66), Alpha=Power(Base=@n67)) {constA=1}
  )
  n69 = LinearInterpolate(
   A=ScalarParameter() {defaultValue=0.6, parameterName=Detail Lerp Low}
   B=ScalarParameter() {defaultValue=1.75, parameterName=Detail Lerp High}
   Alpha=Power(Base=NamedRerouteUsage(), Exp=ScalarParameter() {defaultValue=0.5, parameterName=Detail Power})
  ) {constA=0.5}
  n70 = NamedRerouteUsage()
  n71 = LinearInterpolate(
   A=VectorParameter() {group=Slurp, parameterName=Slurp Color Low, defaultValue=<rgba 0 0.0521 0.1042 1>}.RGB
   B=LinearInterpolate(
    A=VectorParameter() {group=Slurp, parameterName=Surp Color High 0, defaultValue=<rgba 0.0052 0.685 1 1>}.RGB
    B=VectorParameter() {group=Slurp, parameterName=Surp Color High 1, defaultValue=<rgba 0 1 0.3167 1>}.RGB
    Alpha=Saturate(None=@n64)
   )
   Alpha=Multiply(A=Power(Base=@n65, Exp=@n68), B=LinearInterpolate(A=@n69, Alpha=@n70))
  )
  n72 = ScalarParameter() {defaultValue=6, group=Slurp, parameterName=Emissive Intensity}
  n73 = Constant() {r=1}
  n74 = ScalarParameter() {defaultValue=0.5, group=Slurp, parameterName=Night Emissive Multi}
  n75 = SetMaterialAttributes(
   BaseColor=@n71
   EmissiveColor=Multiply(
    A=Multiply(A=@n71, B=@n72)
    B=MaterialFunctionCall(Color_Morning=@n73, Color_Day=@n73, Color_Evening=@n73, Color_Night=@n74) {materialFunction=&Color_Time_of_Day}.Result
   )
   Specular=@n63
   Roughness=@n63
   Normal=@n45
  ) {attributeSetTypes [
    #1
    #2
    #3
    #4
    #5
   ]}
  NamedRerouteDeclaration(None=@n75) {name=SlurpMaterial, nodeColor=<rgba 0.9453 0 1 1>}
  n76 = ScalarParameter() {defaultValue=0.0025, group=Slurp, parameterName=Mask Range Min}
  n77 = ScalarParameter() {defaultValue=0.35, group=Slurp, parameterName=Mask Range Max}
  n78 = CollectionParameter() {collection=&FortniteMaterialParameters, parameterName=SlurpyCenter}
  n79 = ComponentMask(None=CollectionParameter() {collection=&FortniteMaterialParameters, parameterName=SlurpyRadius}) {b=true, g=true, r=true}
  n80 = CollectionParameter() {
   collection=&FortniteMaterialParameters
   parameterName=SlurpTextureMaskLocation
  }
  n81 = Add(A=WorldPosition().XYZ, B=OneMinus(None=ComponentMask(None=@n80) {b=true, g=true, r=true}))
  n82 = TextureObjectParameter() {parameterName=JuiceMask, samplerSource=SSM_Clamp_WorldGroupSettings, texture=&/Clyde_OilSpillV3}
  n83 = MaterialFunctionCall!(
   Export Float 4=StaticBoolParameter() {defaultValue=true, parameterName="Export Float 4 (SB)"}
   TextureObject=@n82
   TextureSize=CollectionParameter() {collection=&FortniteMaterialParameters, parameterName=JuiceMaskSize}
   WorldPosition=@n81
   World Space Normal=@n81
   RotZAxis
   RotXAxis
   RotYAxis
   ProjectionTransitionContrast
  ) {materialFunction=&/MF_WorldAlignedTexture_NoWrap}
  n84 = Subtract(
   A=SphereMask(A=WorldPosition().XYZ, B=ComponentMask(None=@n78) {b=true, g=true, r=true}, Radius=Multiply(A=@n79) {constB=1.5}, Hardness=Constant() {r=5}) {hardnessPercent=0.1}
   B=OneMinus(None=ComponentMask(None=@n83.XYZ Texture) {r=true})
  )
  n85 = Power(
   Base=Saturate(None=InverseLinearInterpolate(A=@n76, B=@n77, Value=@n84))
   Exp=ScalarParameter() {defaultValue=0.3, group=Slurp, parameterName=Mask Power}
  )
  n86 = ScalarParameter() {defaultValue=1, group=Slurp, parameterName=Mask Multiplier}
  NamedRerouteDeclaration(None=Saturate(None=Multiply(A=@n85, B=@n86))) {name=SlurpMask, nodeColor=<rgba 0.0442 0.7756 1 1>}
  NamedRerouteDeclaration(None=MaterialFunctionCall(
    Fade Length=ScalarParameter() {defaultValue=5000, parameterName=Camera Effects Fade Length}
    Fade Offset=ScalarParameter() {defaultValue=1000, parameterName=Camera Effects Fade Offset}
    For Vertex Shader=StaticBool() {value=true}
   ) {materialFunction=&CameraDepthFade}.Result) {name=Camera Effects Fade, nodeColor=<rgba 0.3594 0 1 1>}
  n87 = Normalize(VectorInput=VectorParameter() {parameterName=Wave Direction, defaultValue=<rgba 0.5 0.5 0 1>}.RGB)
  n88 = WorldPosition()
  n89 = Subtract(
   A=MaterialFunctionCall!(Plane Normal=NamedRerouteDeclaration(None=@n87) {name=Wave Direction, nodeColor=<rgba 1 0.3516 0 1>}, Vector=@n88.XYZ) {materialFunction=/Engine/Functions/Engine_MaterialFunctions02/Math/ProjectVectorOntoPlane}.Result
   B=@n88.XYZ
  )
  n90 = NamedRerouteUsage()
  n91 = InverseLinearInterpolate(B=ScalarParameter() {defaultValue=1, group=Slurp, parameterName=Wave Shoreline Width}, Value=@n90)
  n92 = LinearInterpolate(
   A=ScalarParameter() {defaultValue=0.988, group=Slurp, parameterName=Wave Shoreline Displacement Strength}
   Alpha=Power(Base=Saturate(None=@n91), Exp=ScalarParameter() {defaultValue=0.5, group=Slurp, parameterName=Wave Shoreline Power})
  ) {constA=0.9}
  n93 = ScalarParameter() {defaultValue=0, group=Slurp, parameterName=Wave Frequency}
  n94 = Time()
  n95 = ScalarParameter() {group=Slurp, parameterName=Wave Time Offset}
  n96 = CollectionParameter() {collection=&FortniteMaterialParameters, parameterName=Main_Phase_Tracker}
  MaterialExpressionScalarParameter_7 = ScalarParameter() {defaultValue=0.1875, group=Slurp, parameterName=Wave Speed}
  n97 = Switch!(
   SwitchValue=@n96
   Default=@MaterialExpressionScalarParameter_7
   Empty=@MaterialExpressionScalarParameter_7
   Phase1=@MaterialExpressionScalarParameter_7
   Phase2=@MaterialExpressionScalarParameter_7
   Phase3=@MaterialExpressionScalarParameter_7
  ) {inputs [
    {
     inputName=Empty
     input {
      expression=&self:MaterialExpressionScalarParameter_7
      inputName=None
      mask=0
      maskA=0
      maskB=0
      maskG=0
      maskR=0
      outputIndex=0
     }
    }
    {
     inputName=Phase1
     input {
      expression=&self:MaterialExpressionScalarParameter_7
      inputName=None
      mask=0
      maskA=0
      maskB=0
      maskG=0
      maskR=0
      outputIndex=0
     }
    }
    {
     inputName=Phase2
     input {
      expression=&self:MaterialExpressionScalarParameter_7
      inputName=None
      mask=0
      maskA=0
      maskB=0
      maskG=0
      maskR=0
      outputIndex=0
     }
    }
    {
     inputName=Phase3
     input {
      expression=&self:MaterialExpressionScalarParameter_7
      inputName=None
      mask=0
      maskA=0
      maskB=0
      maskG=0
      maskR=0
      outputIndex=0
     }
    }
   ]}
  n98 = Add(A=Multiply(A=Length(None=@n89) {is=class MaterialExpressionLength}, B=Multiply(A=@n92, B=@n93)), B=Multiply(A=Add(A=@n94, B=@n95), B=@n97))
  n99 = Normalize(VectorInput=CrossProduct(A=@n87, B=Constant3Vector() {constant=<rgba 0 0 1 0>}))
  n100 = VectorParameter() {parameterName=Wave Sphere Center, defaultValue=<rgba -35410 -110415 0 1>}
  n101 = Subtract(A=@n88.XYZ, B=@n100.RGB)
  n102 = Length(None=Subtract(
    A=MaterialFunctionCall(Plane Normal=@n99, Vector=@n101) {materialFunction=&ProjectVectorOntoPlane}.Result
    B=@n101
   ))
  n103 = ScalarParameter() {defaultValue=0, parameterName=Wave Curve Tiling}
  n104 = CurveAtlasRowParameter(CurveTime=Saturate(None=Multiply(A=@n102, B=@n103))) {
   atlas=/Clyde/Athena/Apollo/Water/Textures/KelpLinen/CA_KelpLinen
   curve=/Clyde/Athena/Apollo/Water/Textures/KelpLinen/C_KelpLinen_WaveCurve
   defaultValue=1
   group=Slurp
   parameterName=Wave Curve Curve
  }
  n105 = Multiply(A=@n104.R, B=Multiply(A=ScalarParameter() {defaultValue=1, parameterName=Wave Curve Strength}) {constB=-1})
  n106 = CurveAtlasRowParameter(CurveTime=Frac(None=Add(A=@n98, B=@n105))) {
   atlas=&CA_KelpLinen
   curve=/Clyde/Athena/Apollo/Water/Textures/KelpLinen/C_KelpLinen_WaveShape
   group=Slurp
   parameterName=Wave Shape Curve
  }
  n107 = ScalarParameter() {defaultValue=0.2, group=Slurp, parameterName=Wave Shoreline Dampening Width}
  n108 = Power(
   Base=Saturate(None=InverseLinearInterpolate(B=@n107, Value=@n90))
   Exp=ScalarParameter() {defaultValue=0.8, group=Slurp, parameterName=Wave Shoreline Dampening Power}
  )
  n109 = InverseLinearInterpolate(
   B=ScalarParameter() {defaultValue=0.8, group=Slurp, parameterName=Wave Sphere Cutoff High}
   Value=SphereMask(A=WorldPosition().XYZ, B=@n100.RGB, Radius=ScalarParameter() {defaultValue=17500, group=Slurp, parameterName=Wave Sphere Radius}) {hardnessPercent=0}
  )
  NamedRerouteDeclaration(None=Multiply(A=Multiply(A=@n106.R, B=@n108), B=Saturate(None=@n109))) {name=Waves, nodeColor=<rgba 1 0.3516 0 1>}
  NamedRerouteDeclaration(None=@n55) {name=Wave Distortion, nodeColor=<rgba 1 0.3516 0 1>}
  n110 = FunctionInput(Preview=MaterialFunctionCall!(Texture, Position) {materialFunction=/FortWater/Materials/Functions/WaterBodyData}.Water Depth) {bUsePreviewValueAsDefault=true, inputName=Water Depth, inputType=FunctionInput_Scalar, sortPriority=1}
  n111 = ScalarParameter() {defaultValue=1200, group=Slurp, parameterName=Water Shoreline Depth}
  NamedRerouteDeclaration(None=Saturate(None=Divide(A=@n110, B=@n111))) {name=Shoreline Depth, nodeColor=<rgba 0.9922 0 1 1>}
  NamedRerouteDeclaration(None=ComponentMask(None=@n16.RGB) {g=true}) {name=Alpha Noise, nodeColor=<rgba 0 1 0.4609 1>}
  n112 = Add(A=Add(A=@n51, B=@n52), B=@n53)
  NamedRerouteDeclaration(None=ComponentMask(None=TextureSample(UVs=@n112, Tex=@n46) {samplerSource=SSM_Wrap_WorldGroupSettings}.RGB) {b=true}) {name=Glow Noise, nodeColor=<rgba 0 1 0.4609 1>}
  OneMinus(None=@n62)
  ScalarParameter() {defaultValue=0.75, group=Slurp, parameterName=Flow Map Alpha Strength Old}
  ScalarParameter() {defaultValue=2, group=Slurp, parameterName=Flow Map Glow Old}
 ]}
