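# Water material function that adds panning far-wave normals over an input normal and fades out WPO by a far mask.
# Two samples of one tiling normal, scaled by golden ratio, pan on water time in world or UV space, with optional object-space direction.
# Biome masks can vary scale and intensity; the far mask, with a close-up minimum, drives the blend.
# Low quality, forward and mobile paths skip that blend, using only the flattened panned normal.
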MaterialFunction /Clyde/Athena/Apollo/Water/Materials/MF_WaterTwoTexturePanner_Clyde
graph {nodes [
  Constant3Vector() {constant=<rgba 0 0 1 0>}
  Constant3Vector() {constant=<rgba 0.1 0.1 1 0>}
  Constant() {r=1}
  n1 = TextureCoordinate()
  n2 = ScalarParameter() {defaultValue=1, group=Distance Blending, parameterName=Pan_UV_size}
  n3 = WorldPosition() {worldPositionShaderOffset=WPT_ExcludeAllShaderOffsets}
  n4 = ScalarParameter() {defaultValue=8000, group=Distance Blending, parameterName=FarWaveNormalScale}
  n5 = FunctionInput(Preview=Constant()) {bUsePreviewValueAsDefault=true, inputName=BiomeMask, inputType=FunctionInput_Scalar}
  n6 = Constant() {r=0.25}
  n7 = Constant() {r=0.75}
  n8 = StaticSwitchParameter(
   True=LinearInterpolate(
    A=@n4
    B=ScalarParameter() {defaultValue=8000, group=Distance Blending, parameterName=FarWaveNormalScaleBiome}
    Alpha=MaterialFunctionCall(Input=@n5, Input Low=@n6, Input High=@n7) {materialFunction=&RemapValue_0-1}.Result
   )
   False=@n4
  ) {group=Distance Blending, parameterName=UseBiomeWaveNormals}
  n9 = Multiply(A=@n8, B=ScalarParameter() {defaultValue=1, group=Distance Blending, parameterName=FarWaveNormalScaleFactor_LowDetail})
  n10 = QualitySwitch(Default=@n8, Low=@n9)
  n11 = Divide(
   A=StaticSwitchParameter(True=Multiply(A=@n1, B=@n2), False=ComponentMask(None=@n3.XYZ) {g=true, r=true}) {parameterName=Pan in UV Space}
   B=ShadingPathSwitch(Default=@n10, Deferred=@n10, Forward=@n9, Mobile=@n9)
  )
  n12 = MaterialFunctionCall(
   Time=MaterialFunctionCall() {materialFunction=/Water/Materials/Functions/GetWaterTime}.WaterTime
   Speed=ScalarParameter() {defaultValue=1.6, group=Distance Blending, parameterName=FarWaveNormalSpeed}
  ) {materialFunction=&TimeWithSpeedVariable}
  n13 = Normalize(VectorInput=Transform(None=VectorParameter() {group=Distance Blending, parameterName=Distance ObjectSpace Panner, defaultValue=<rgba 1 0 0 0>}.RGB) {transformSourceType=TRANSFORMSOURCE_Local})
  n14 = Multiply(A=@n13, B=ScalarParameter() {defaultValue=0.08, group=Distance Blending, parameterName=Distance Object Panner_1})
  n15 = VectorParameter() {group=Distance Blending, parameterName=Distance Normal Panner 1, defaultValue=<rgba 0.05 0.02 0 0>}
  n16 = ScalarParameter() {defaultValue=1, group=Distance Blending, parameterName=FarMipBias}
  n17 = TextureSampleParameter2D!(
   UVs=Panner(Coordinate=@n11, Time=@n12.Time, Speed=StaticSwitchParameter(True=@n14, False=@n15.RGB) {parameterName=Pan in Object Space})
   Bias=@n16
   Apply View MipBias
  ) {
   group=Distance Blending
   mipValueMode=TMVM_MipBias
   parameterName=FarWaveNormalHeight
   samplerSource=SSM_Wrap_WorldGroupSettings
   samplerType=SAMPLERTYPE_LinearColor
   texture=/Game/Global/Textures/Default/Tileable_Patterns/T_Water_TilingNormal_With_Height_02_Softened
  }
  n18 = ComponentMask(None=@n17.RGB) {g=true, r=true}
  n19 = Multiply(A=@n13, B=ScalarParameter() {defaultValue=0.064, group=Distance Blending, parameterName=Distance Object Panner_2})
  n20 = VectorParameter() {group=Distance Blending, parameterName=Distance Normal Panner 2, defaultValue=<rgba -0.01 -0.06 0 0>}
  n21 = TextureSampleParameter2D(
   UVs=Panner(
    Coordinate=Multiply(A=@n11) {constB=1.618}
    Time=@n12.Time
    Speed=StaticSwitchParameter(True=@n19, False=@n20.RGB) {parameterName=Pan in Object Space}
   )
   Bias=@n16
  ) {
   group=Distance Blending
   mipValueMode=TMVM_MipBias
   parameterName=FarWaveNormalHeight
   samplerSource=SSM_Wrap_WorldGroupSettings
   samplerType=SAMPLERTYPE_LinearColor
   texture=&T_Water_TilingNormal_With_Height_02_Softened
  }
  n22 = ComponentMask(None=@n21.RGB) {g=true, r=true}
  n23 = AppendVector(A=Add(A=ConstantBiasScale(None=@n18) {bias=-0.5, scale=2}, B=ConstantBiasScale(None=@n22) {bias=-0.5, scale=2}), B=Constant())
  n24 = ScalarParameter() {defaultValue=0.05, group=Distance Blending, parameterName=FarWavelNormalIntensity}
  n25 = Constant()
  n26 = MaterialFunctionCall(Input=@n5, Input Low=Constant(), Input High=@n6) {materialFunction=&RemapValue_0-1}
  n27 = Constant() {r=1}
  n28 = StaticSwitchParameter(
   True=LinearInterpolate(
    A=LinearInterpolate(A=@n24, B=@n25, Alpha=@n26.Result)
    B=ScalarParameter() {defaultValue=0.05, group=Distance Blending, parameterName=FarWaveNormalIntensityBiome}
    Alpha=MaterialFunctionCall(Input=@n5, Input Low=@n7, Input High=@n27) {materialFunction=&RemapValue_0-1}.Result
   )
   False=@n24
  ) {group=Distance Blending, parameterName=UseBiomeWaveNormals}
  n29 = Constant()
  n30 = FunctionInput() {bUsePreviewValueAsDefault=true, inputName=Additional Mask, inputType=FunctionInput_Scalar, sortPriority=3, previewValue=<1 0 0 1>}
  n31 = FunctionOutput(None=Multiply(
    A=FunctionInput(Preview=@n29) {bUsePreviewValueAsDefault=true, inputName=Far Mask Override, inputType=FunctionInput_Scalar, sortPriority=2}
    B=@n30
   )) {outputName=Far Mask, sortPriority=3}
  n32 = LinearInterpolate(
   A=FunctionInput() {bUsePreviewValueAsDefault=true, inputName=Normal, previewValue=<0 0 1 1>}
   B=MaterialFunctionCall(Normal=@n23, Flatness=OneMinus(None=@n28)) {materialFunction=&FlattenNormal}.Result
   Alpha=Max(A=@n31, B=ScalarParameter() {defaultValue=0.15, parameterName=DetailNormal Up Close Amount})
  )
  n33 = ScalarParameter() {defaultValue=0.05, group=Distance Blending, parameterName=FarWavelNormalIntensity_LowDetail}
  n34 = FunctionInput() {bUsePreviewValueAsDefault=true, inputName=LowDetailBiome_Mask, inputType=FunctionInput_Scalar, sortPriority=8, previewValue=<1 0 0 1>}
  n35 = MaterialFunctionCall(
   Normal=@n23
   Flatness=LinearInterpolate(
    A=OneMinus(None=@n33)
    B=ScalarParameter() {defaultValue=0.99, group=Distance Blending, parameterName=FarWavelNormalIntensity_Biome_LowDetail}
    Alpha=Saturate(None=@n34)
   )
  ) {materialFunction=&FlattenNormal}
  n36 = QualitySwitch(Default=@n32, Low=@n35.Result, High=@n32, Medium=@n32)
  FunctionOutput(None=LinearInterpolate(
    A=Constant3Vector() {constant=<rgba 0 0 1 0>}
    B=ShadingPathSwitch(Default=@n36, Deferred=@n36, Forward=@n35.Result, Mobile=@n35.Result)
    Alpha=@n30
   )) {outputName=Normal}
  FunctionOutput(None=Multiply(A=FunctionInput() {bUsePreviewValueAsDefault=true, inputName=WPO, sortPriority=1}, B=OneMinus(None=@n31))) {outputName=WPO, sortPriority=1}
 ]}
