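# Material function that computes a ring-shaped mask around an X/Y world location, apparently for kelp wave effects in water.
# World-space XY distance feeds two SmoothSteps with edges offset by Thickness and Falloff, driven separately by Time and Time_0.
# The second step, widened by inner-falloff parameters, is subtracted from the strength-scaled first.
# A Power with a fade exponent shapes the result, with tuning parameters grouped under Kelp_Linen.
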MaterialFunction /Clyde/Athena/Apollo/Water/Materials/MF_KelpWaves_Rings
graph {nodes [
  n1 = Distance(
   A=WorldPosition().XY
   B=AppendVector(
    A=FunctionInput() {inputName=X Location, inputType=FunctionInput_Scalar, sortPriority=10}
    B=FunctionInput() {inputName=Y Location, inputType=FunctionInput_Scalar, sortPriority=20}
   )
  )
  n2 = FunctionInput() {inputName=Thickness, inputType=FunctionInput_Scalar, sortPriority=40}
  n3 = Subtract(A=FunctionInput() {inputName=Time, inputType=FunctionInput_Scalar}, B=Divide(A=@n2))
  n4 = FunctionInput() {inputName=Falloff, inputType=FunctionInput_Scalar, sortPriority=50}
  n5 = Multiply(A=@n4)
  n6 = Multiply(
   A=MaterialFunctionCall(Alpha=@n1, Min=Subtract(A=@n3, B=@n5), Max=@n3) {materialFunction=&SmoothStep}.Result
   B=ScalarParameter() {defaultValue=0.424, group=Kelp_Linen, parameterName=Outer_Ring_Str}
  )
  n7 = ScalarParameter() {defaultValue=6322.7842, group=Kelp_Linen, parameterName=Falloff_Inner}
  n8 = Add(A=FunctionInput() {inputName=Time_0, inputType=FunctionInput_Scalar, sortPriority=30}, B=Divide(A=Add(A=@n2, B=@n7)))
  n9 = Add(A=@n4, B=@n7)
  n10 = ScalarParameter() {defaultValue=0.2, group=Kelp_Linen, parameterName=FalloffMulti}
  n11 = MaterialFunctionCall(Alpha=@n1, Min=@n8, Max=Add(A=@n8, B=Multiply(A=@n9, B=@n10))) {materialFunction=&SmoothStep}
  FunctionOutput(None=Power(Base=Subtract(A=@n6, B=@n11.Result), Exp=ScalarParameter() {defaultValue=2, group=Kelp_Linen, parameterName=Kelp_Fade_Pow})) {outputName=Subtract}
 ]}
