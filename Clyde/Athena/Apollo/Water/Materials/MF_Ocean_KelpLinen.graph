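# Ocean water material function that layers a pulsing "Kelp Linen" effect over incoming material attributes, with vertex-normal wave offset and optional time-of-day-scaled emissive.
# Four phase-staggered rings, driven by looping time or a manual lerp, form the pulse, which also yields a foam mask via panned world-space texture.
# A radial fade minus island box masks is built for localization.
# Masked blends output material attributes plus water scattering and absorption.
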MaterialFunction /Clyde/Athena/Apollo/Water/Materials/MF_Ocean_KelpLinen
graph {nodes [
  n1 = ScalarParameter() {defaultValue=4, group=Kelp_Linen, parameterName=Circle_Size}
  n2 = Constant() {r=-0.5}
  n3 = MaterialFunctionCall(Float2=Add(A=AppendVector(A=@n2, B=@n2), B=TextureCoordinate())) {materialFunction=&BreakOutFloat2Components}
  n4 = Multiply(A=@n1, B=@n3.R)
  n5 = Multiply(A=@n3.G, B=@n1)
  n6 = Add(A=Multiply(A=@n4, B=@n4), B=Multiply(A=@n5, B=@n5))
  n7 = Clamp(None=@n6)
  Add(A=@n7)
  Clamp(None=@n6)
  CollectionParameter() {collection=/Game/Athena/Events/AudioAnalysis/MPC_Audio_Reactive_Parameters}
  Constant4Vector() {constant=<rgba 0.2135 0.0369 0.0602 1>}
  Constant4Vector() {constant=<rgba 0.4323 0.0327 0.2989 1>}
  Constant() {r=0.0002}
  Constant() {r=0.7}
  n8 = Clamp(None=Multiply(A=NamedRerouteUsage(), B=NamedRerouteUsage()))
  n9 = BlendMaterialAttributes(A=NamedRerouteUsage(), B=NamedRerouteUsage(), Alpha=@n8)
  n10 = NamedRerouteUsage()
  FunctionOutput(None=StaticSwitchParameter(
    True=SetMaterialAttributes!(MaterialAttributes=@n9, EmissiveColor=@n10) {attributeSetTypes [#1]}
    False=@n9
   ) {group=Kelp_Linen, parameterName=Fade_Debug, sortPriority=31})
  n11 = VectorParameter() {
   group=Kelp_Linen
   parameterName=Kelp_Controls
   channelNames {a=Thickness, b=Radius, g=Y Location, r=X location}
   defaultValue=<rgba -97189.6406 -186678.0625 140835.5156 -547.2886>
  }
  n12 = StaticSwitchParameter(
   True=Multiply(A=Time(), B=ScalarParameter() {defaultValue=0.1, desc=Sudo how long the pulse takes, group=Kelp_Linen, parameterName=Time_Multi}) {constB=0.2}
   False=ScalarParameter() {group=Kelp_Linen, parameterName=Kelp_Lerp, sliderMax=1}
  ) {group=Kelp_Linen, parameterName=Loop?, sortPriority=31}
  n13 = LinearInterpolate(B=@n11.Radius, Alpha=Frac(None=@n12))
  n14 = Constant() {r=10000}
  n15 = Constant() {r=0.25}
  n16 = Multiply(A=@n15)
  n17 = LinearInterpolate(B=@n11.Radius, Alpha=Frac(None=Add(A=@n12, B=@n16)))
  n18 = Multiply(A=@n15) {constB=2}
  n19 = LinearInterpolate(B=@n11.Radius, Alpha=Frac(None=Add(A=@n12, B=@n18)))
  n20 = Add(
   A=Add(
    A=MaterialFunctionCall!(Time=@n13, X Location=@n11.X location, Y Location=@n11.X location, Time_0=@n13, Thickness=@n11.X location, Falloff=@n14) {materialFunction=&/MF_KelpWaves_Rings}.Subtract
    B=MaterialFunctionCall(Time=@n17, X Location=@n11.X location, Y Location=@n11.X location, Time_0=@n17, Thickness=@n11.X location, Falloff=@n14) {materialFunction=&MF_KelpWaves_Rings}.Subtract
   )
   B=MaterialFunctionCall(Time=@n19, X Location=@n11.X location, Y Location=@n11.X location, Time_0=@n19, Thickness=@n11.X location, Falloff=@n14) {materialFunction=&MF_KelpWaves_Rings}.Subtract
  )
  n21 = Multiply(A=@n15) {constB=3}
  n22 = LinearInterpolate(B=@n11.Radius, Alpha=Frac(None=Add(A=@n12, B=@n21)))
  n23 = MaterialFunctionCall(Time=@n22, X Location=@n11.X location, Y Location=@n11.X location, Time_0=@n22, Thickness=@n11.X location, Falloff=@n14) {materialFunction=&MF_KelpWaves_Rings}
  n24 = NamedRerouteDeclaration(None=Saturate(None=Add(A=@n20, B=@n23.Subtract))) {name=Pulse_Mask, nodeColor=<rgba 0 1 0.2734 1>}
  n25 = ComponentMask(None=WorldPosition() {worldPositionShaderOffset=WPT_ExcludeAllShaderOffsets}.XYZ) {g=true, r=true}
  n26 = ScalarParameter() {defaultValue=20000, group=Kelp_Linen, parameterName=Kelp_FoamScale}
  n27 = TextureSample(UVs=Panner(Coordinate=Divide(A=@n25, B=@n26)) {speedY=-0.0002}) {samplerType=SAMPLERTYPE_Masks, texture=/FortWater/Textures/T_OceanFoam_New_Mask}
  n28 = Constant() {r=46223.1758}
  n29 = Power(Base=MaterialFunctionCall(
    UVs=WorldPosition().XY
    CenterPosition=AppendVector(A=@n11.X location, B=@n11.X location)
    Radius=Subtract(A=@n11.Radius, B=@n28)
    Density=Constant() {r=20}
   ) {materialFunction=&RadialGradientExponential}.RadialGradientExponential)
  FunctionOutput(None=Multiply(A=Multiply(A=@n24, B=@n27.R), B=Multiply(A=@n29))) {description=Used to control foam since we are controlling it manually here, outputName=Pulse_Mask, sortPriority=1}
  n30 = VectorParameter() {group=Kelp_Linen, parameterName=Kelp_Scattering, defaultValue=<rgba 0.4235 0 0.5677 1.3396>}
  n31 = Divide(A=@n30.A) {constB=100}
  n32 = VectorParameter() {group=Kelp_Linen, parameterName=Kelp_Scattering_2, defaultValue=<rgba 0.5729 0.074 0.3794 1.3396>}
  n33 = Divide(A=@n32.A) {constB=100}
  n34 = VectorParameter() {group=Kelp_Linen, parameterName=Kelp_Absorption, defaultValue=<rgba 0.5208 0.1821 0.5101 259.5518>}
  n35 = VectorParameter() {group=Kelp_Linen, parameterName=Kelp_Absorption2, defaultValue=<rgba 0.3646 0.1275 0.357 259.5518>}
  n36 = MaterialFunctionCall!(
   Scattering A=FunctionInput() {inputName=Scattering}
   Scattering B=LinearInterpolate(A=Multiply(A=@n30.RGB, B=@n31), B=Multiply(A=@n32.RGB, B=@n33), Alpha=NamedRerouteUsage())
   Absorption A=FunctionInput() {inputName=Absorption}
   Absorption B=LinearInterpolate(A=Multiply(A=@n34.RGB, B=@n34.RGB), B=Multiply(A=@n35.RGB, B=@n35.RGB), Alpha=NamedRerouteUsage())
   Alpha=@n8
  ) {materialFunction=/FortWater/Materials/Functions/BlendWaterShadingValues}
  FunctionOutput(None=@n36.Absorption) {outputName=Absorbtion, sortPriority=5}
  FunctionOutput(None=@n36.Scattering) {outputName=Scatter, sortPriority=6}
  n37 = ScalarParameter() {defaultValue=2, group=Kelp_Linen, parameterName=Pow}
  Multiply(
   A=MaterialFunctionCall!(Width, Sharpness, Offset, Direction Switch, Compare, Input Coordinates=Panner(Coordinate=@n7) {speedY=-1}) {materialFunction=/Engine/Functions/Engine_MaterialFunctions03/Procedurals/GeneratedBand}.Result
   B=OneMinus(None=Power(Base=@n6, Exp=@n37))
  )
  n38 = GetMaterialAttributes(None=NamedRerouteUsage())
  n39 = Constant() {r=1}
  n40 = ScalarParameter() {defaultValue=1000, group=Kelp_Linen, parameterName=Wave_Height}
  n41 = Add(
   A=NamedRerouteUsage()
   B=Multiply(
    A=VertexNormalWS()
    B=MaterialFunctionCall(R=@n39, G=@n39, B=@n40) {materialFunction=&AppendMany}.RGB
   )
  )
  n42 = TextureSample(UVs=Multiply(A=TextureCoordinate(), B=Constant() {r=0})) {samplerType=SAMPLERTYPE_Masks, texture=/Game/Global/Textures/Default/Masks_and_Gradients/Swamp_Mask}
  n43 = LinearInterpolate(
   A=Constant3Vector() {constant=<rgba 0 9.4843 10 1>}
   B=Constant3Vector() {constant=<rgba 0 4.4882 10 1>}
   Alpha=Power(Base=@n42.G, Exp=Constant() {r=1.0507})
  )
  n44 = Constant() {r=1}
  n45 = Multiply(
   A=Multiply(A=@n43, B=Constant() {r=4})
   B=MaterialFunctionCall(Color_Morning=@n44, Color_Day=@n44, Color_Evening=@n44, Color_Night=Constant() {r=0.5}) {materialFunction=&Color_Time_of_Day}.Result
  )
  NamedRerouteDeclaration(None=StaticSwitchParameter(
    True=SetMaterialAttributes!(MaterialAttributes=@n38.MaterialAttributes, WorldPositionOffset=@n41, EmissiveColor=@n45) {attributeSetTypes [#2, #1]}
    False=SetMaterialAttributes!(MaterialAttributes=@n38.MaterialAttributes, WorldPositionOffset=@n41) {attributeSetTypes [#2]}
   ) {defaultValue=true, group=Kelp_Linen, parameterName=Use Emissive, sortPriority=31}) {name=Kelp Linen Material, nodeColor=<rgba 0 0.3672 1 1>}
  n46 = GetMaterialAttributes(None=FunctionInput() {inputType=FunctionInput_MaterialAttributes, sortPriority=-1}) {attributeGetTypes [#2, #3]}
  NamedRerouteDeclaration(None=@n46.WorldPositionOffset) {name=Baseline_Wave, nodeColor=<rgba 0 1 0.5547 1>}
  NamedRerouteDeclaration(None=SetMaterialAttributes!(MaterialAttributes=@n46.MaterialAttributes, BaseColor=NamedRerouteUsage()) {attributeSetTypes [#3]}) {name=Base Material, nodeColor=<rgba 0.875 0 1 1>}
  n47 = WorldPosition()
  n48 = MaterialFunctionCall(
   UVs=@n47.XY
   CenterPosition=AppendVector(
    A=ScalarParameter() {defaultValue=-78940.2891, group=Kelp_Linen, parameterName=Fade_X}
    B=ScalarParameter() {defaultValue=-160641.3438, group=Kelp_Linen, parameterName=Fade_Y}
   )
   Radius=ScalarParameter() {defaultValue=68424.4453, group=Kelp_Linen, parameterName=Fade_Radius}
   Density=ScalarParameter() {defaultValue=4.867, group=Kelp_Linen, parameterName=Fade_Density}
  ) {materialFunction=&RadialGradientExponential}
  n49 = AppendVector(A=Constant() {r=-228051.5938}, B=Constant() {r=-634946.1875})
  n50 = AppendVector(A=Constant() {r=566757.3125}, B=Constant() {r=379174.3438})
  n51 = Constant() {r=50161.5742}
  n52 = AppendVector(
   A=ScalarParameter() {defaultValue=-228051.5938, group=Kelp_Linen, parameterName=Fade_X_Box}
   B=ScalarParameter() {defaultValue=-634946.1875, group=Kelp_Linen, parameterName=Fade_Y_Box}
  )
  n53 = AppendVector(
   A=ScalarParameter() {defaultValue=566757.3125, group=Kelp_Linen, parameterName=Fade_X_Box_Radius}
   B=ScalarParameter() {defaultValue=379174.3438, group=Kelp_Linen, parameterName=Fade_Y_Box_Radius}
  )
  n54 = ScalarParameter() {defaultValue=50161.5742, group=Kelp_Linen, parameterName=Box_Falloff}
  n55 = Saturate(None=Subtract(
    A=MaterialFunctionCall!(A=@n47.XY, B=@n49, Bounds=@n50, Edge Falloff=@n51) {desc=island locations, materialFunction=/Engine/Functions/Engine_MaterialFunctions01/Gradient/BoxMask-2D}.Result
    B=MaterialFunctionCall(A=@n47.XY, B=@n52, Bounds=@n53, Edge Falloff=@n54) {desc=island adjustments, materialFunction=&BoxMask-2D}.Result
   ))
  NamedRerouteDeclaration(None=Saturate(None=Subtract(A=@n48.RadialGradientExponential, B=@n55))) {name=Fade_Radius, nodeColor=<rgba 0.7109 1 0 1>}
  NamedRerouteDeclaration(None=@n42.G) {name=noise, nodeColor=<rgba 1 0.7734 0 1>}
  n56 = VectorParameter() {group=Foam and Pondscum, parameterName=Foam2 Color 1, defaultValue=<rgba 0.319 0.9302 1 0.4>}
  n57 = VectorParameter() {group=Foam and Pondscum, parameterName=Foam2 Color 2, defaultValue=<rgba 0.817 0.9083 1 0.38>}
  n58 = Multiply(A=@n57.RGB, B=@n57.RGB)
  n59 = LinearInterpolate(A=Multiply(A=@n56.RGB, B=@n56.RGB), B=@n58) {constAlpha=0}
  n60 = QualitySwitch(Default=@n59, Low=@n58, High=@n59, Medium=@n58)
  NamedRerouteDeclaration(None=ShadingPathSwitch(Default=@n60, Deferred=@n60, Forward=@n58, Mobile=@n58)) {name=Foam, nodeColor=<rgba 0.2188 0 1 1>}
  Reroute(None=@n43)
  ScalarParameter() {defaultValue=13851.6396, group=Kelp_Linen, parameterName=falloff}
  ScalarParameter() {defaultValue=1, group=Kelp_Linen, parameterName=Kelp_Emissive Intensity}
  VectorParameter() {group=Kelp_Linen, parameterName=Kelp_Color_A, defaultValue=<rgba 0.0008 0.1042 0.0874 1>}
  VectorParameter() {group=Kelp_Linen, parameterName=Kelp_Color_B, defaultValue=<rgba 0.0076 0.974 0.8173 1>}
 ]}
