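# Two-sided Single Layer Water material for the far ocean, usable on water bodies, spline meshes and instanced meshes.
# A two-texture panner supplies normals, flipped on back faces, alongside parameter-driven color, scattering and squared absorption.
# MF_Ocean_KelpLinen provides alternate attributes, scattering and absorption, with a Kelp_Active mask built from SlurpActive and a location fade.
# Quality and shading-path switches substitute simpler specular and roughness values on lower-end paths.
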Material /Clyde/Athena/Apollo/Water/Materials/M_Water_Ocean_02_NewModel_Far
graph {
 properties {
  allowTranslucentCustomDepthWrites=true
  bUseMaterialAttributes=true
  bUsedWithInstancedStaticMeshes=true
  bUsedWithSplineMeshes=true
  bUsedWithWater=true
  pixelDepthOffsetMode=PDOM_Legacy
  shadingModel=MSM_SingleLayerWater
  twoSided=true
  usedShadingModels=SingleLayerWater
 }
 nodes [
  n1 = ScalarParameter() {defaultValue=0.225, group=Water Shading Model, parameterName=Specular}
  n2 = ScalarParameter() {defaultValue=0.13, group=Water Shading Model, parameterName=Specular_LowDetail}
  n3 = MaterialFunctionCall!(
   Normal
   WPO=VectorParameter() {parameterName=PositionOffset, defaultValue=<rgba 0 0 0 1>}.RGB
   Far Mask Override
   Additional Mask
   LowDetailBiome_Mask
  ) {materialFunction=/FortWater/Materials/Functions/WaterTwoTexturePanner}
  n4 = SetMaterialAttributes!(
   MaterialAttributes
   Base Color=VectorParameter() {parameterName=BaseColor}.RGB
   Specular=QualitySwitch(
    Default=MaterialFunctionCall!(InSpecular=@n1, WaterDepth) {materialFunction=/FortWater/Materials/Functions/MF_WaterBodySpecular}.Result
    Low=@n2
    Medium=@n2
   )
   Roughness=ScalarParameter() {defaultValue=0.05, group=Water Shading Model, parameterName=Roughness}
   Opacity=ScalarParameter() {group=Water Shading Model, parameterName=Surface Opacity}
   Normal=@n3.Normal
   World Position Offset=@n3.Normal
  ) {attributeSetTypes [
    #1
    #2
    #3
    #4
    #5
    #6
   ]}
  n5 = GetMaterialAttributes(None=@n4) {attributeGetTypes [
    #1
    #2
    #3
    #4
    #5
    #6
   ]}
  n6 = SetMaterialAttributes!(
   MaterialAttributes
   Base Color=@n5.Base Color
   Metallic=Constant()
   Specular=@n5.Base Color
   Roughness=@n5.Base Color
   Opacity=@n5.Base Color
   Normal=Multiply(A=@n5.Normal, B=TwoSidedSign())
   World Position Offset=@n5.Base Color
  ) {attributeSetTypes [
    #1
    #7
    #2
    #3
    #4
    #5
    #6
   ]}
  n7 = MaterialFunctionCall!(In=@n6, Scattering=NamedRerouteUsage(), Absorption=NamedRerouteUsage()) {materialFunction=&/MF_Ocean_KelpLinen}
  BlendMaterialAttributes(A=@n6, B=@n7.Result, Alpha=NamedRerouteUsage())
  Constant()
  MaterialFunctionCall!(Normal=@n5.Normal) {materialFunction=/FortWater/Materials/Functions/MF_WaveFlatten}
  n8 = VectorParameter() {group=Water Shading Model, parameterName=Absorption, defaultValue=<rgba 1 13.35 20 100>}
  n9 = Multiply(A=@n8.RGB, B=@n8.RGB)
  NamedRerouteDeclaration(None=@n9) {name=base absorption, nodeColor=<rgba 0.3359 0 1 1>}
  n10 = CollectionParameter() {collection=&FortniteMaterialParameters, parameterName=SlurpActive}
  n11 = MaterialFunctionCall() {materialFunction=/ClydeCosmetics/Environments/KelpLinen/MF_KelpLinen_LocationFade}
  NamedRerouteDeclaration(None=Saturate(None=Multiply(A=@n10, B=@n11.Output))) {name=Kelp_Active, nodeColor=<rgba 0 1 0.8125 1>}
  NamedRerouteDeclaration(None=@n7.Scatter) {name=final scattering, nodeColor=<rgba 0 0.6719 1 1>}
  NamedRerouteDeclaration(None=@n7.Absorbtion) {name=final absorption, nodeColor=<rgba 0.9922 1 0 1>}
  n12 = GetMaterialAttributes(None=@n6) {attributeGetTypes [#3]}
  n13 = ScalarParameter() {defaultValue=0.1, group=Water Shading Model, parameterName=RoughnessLowDetail}
  n14 = QualitySwitch(Default=@n12.Roughness, Low=@n13, Medium=@n13)
  SetMaterialAttributes!(MaterialAttributes=@n12.MaterialAttributes, Roughness=ShadingPathSwitch(Default=@n14, Deferred=@n14, Forward=@n13, Mobile=@n13)) {attributeSetTypes [#3]}
  n15 = VectorParameter() {group=Water Shading Model, parameterName=Scattering, defaultValue=<rgba 0.1268 0.1336 0.15 0.5>}
  n16 = Divide(A=@n15.A) {constB=100}
  SingleLayerWaterMaterialOutput!(
   ScatteringCoefficients=NamedRerouteDeclaration(None=Multiply(A=@n15.RGB, B=@n16)) {name=base scattering, nodeColor=<rgba 1 0.6562 0 1>}
   AbsorptionCoefficients=Divide(B=@n9)
   PhaseG=ScalarParameter() {defaultValue=0.3, group=Water Shading Model, parameterName=Anisotropy}
   ColorScaleBehindWater
  )
 ]
}
